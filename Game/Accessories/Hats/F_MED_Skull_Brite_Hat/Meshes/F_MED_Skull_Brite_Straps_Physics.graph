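# Physics asset for the Skull Brite hat's straps on a medium female character, defining bodies and joints for their secondary motion.
# Two five-bone strap chains (A and B) hang from the head with limited swing and twist, using capsules near the root and boxes toward the tips.
# An extra constraint links the two chain roots.
# Head, neck, spine and clavicle capsules give the straps body shapes to collide against, joined by free-rotation constraints.
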PhysicsAsset /Game/Accessories/Hats/F_MED_Skull_Brite_Hat/Meshes/F_MED_Skull_Brite_Straps_Physics
bodies {
 clavicle_l [{shapeType=Capsule, length=17.5235, radius=8.0965, center=<17.8357 0.506 3.7175>, rotation=<rot 90 0 0>}]
 clavicle_r [{shapeType=Capsule, length=17.5235, radius=8.0965, center=<-17.8357 -0.506 -3.7175>, rotation=<rot 90 0 0>}]
 dyn_strapA_01 [{shapeType=Capsule, length=3.7881, radius=1.2777, shapeName=dyn_strapA_01_capsule, center=<3 0 0.65>, rotation=<rot 90 0 0>}]
 dyn_strapA_02 [{shapeType=Capsule, length=3.7881, radius=1.2777, shapeName=dyn_strapA_01_capsule, center=<3 0 0>, rotation=<rot 90 0 0>}]
 dyn_strapA_03 [{shapeType=Box, extentX=7.5, extentY=4, extentZ=2, center=<3.1416 0 -0.3501>}]
 dyn_strapA_04 [{shapeType=Box, extentX=6.5, extentY=4, extentZ=2, center=<1.848 -0.0001 -0.3501>}]
 dyn_strapA_05 [{shapeType=Box, extentX=6, extentY=4, extentZ=1.5, center=<1.4521 -0.0001 -0.3501>}]
 dyn_strapB_01 [{
   shapeType=Capsule
   length=2.9734
   radius=1.2777
   shapeName=dyn_strapA_01_capsule
   center=<2.4852 0 0.35>
   rotation=<rot 82.7367 -0.0001 -0.0001>
  }]
 dyn_strapB_02 [{shapeType=Capsule, length=2.9734, radius=1.2777, shapeName=dyn_strapA_01_capsule, center=<2.4852 0 0>, rotation=<rot 90 0 0>}]
 dyn_strapB_03 [{shapeType=Box, extentX=6, extentY=4, extentZ=2, center=<2.843 0 0.0005>}]
 dyn_strapB_04 [{shapeType=Box, extentX=7, extentY=3.5, extentZ=2, center=<3.0052 0 0.0006>}]
 dyn_strapB_05 [{shapeType=Box, extentX=6, extentY=3.5, extentZ=2, center=<2.252 0 0.0004>}]
 head [
  {shapeType=Capsule, length=3, radius=7.91}
  {shapeType=Capsule, length=5.8867, radius=5.5688, shapeName=head_capsule, center=<2.5352 -3.95 0>}
 ]
 neck_01 [{shapeType=Capsule, length=4.5, radius=8.505, center=<0.391 3.2099 -0.0001>, rotation=<rot 0.0001 -130.0001 90.0001>}]
 spine_03 [{shapeType=Capsule, length=6.005, radius=12.755, center=<10.2891 2.3328 0.0002>, rotation=<rot 0.0002 -90.0005 -0.0003>}]
 spine_05 [
  {shapeType=Capsule, length=20, radius=12.755, center=<0.2606 0.7146 -0.0001>, rotation=<rot 0 -90 0>}
  {shapeType=Capsule, length=6.5, radius=8, center=<-1.8402 -0.6408 10.1882>, rotation=<rot 59.9917 180 -174.9998>}
  {shapeType=Capsule, length=6.25, radius=8, center=<-0.7848 0.208 -9.5631>, rotation=<rot 59.9998 0 5.0003>}
 ]
}
constraints [
 {bone1Name=dyn_strapB_01, bone2Name=head, swing1LimitDegrees=45, swing2LimitDegrees=35, twistLimitDegrees=16}
 {bone1Name=dyn_strapB_02, bone2Name=dyn_strapB_01, swing1LimitDegrees=45, swing2LimitDegrees=35, twistLimitDegrees=16}
 {bone1Name=dyn_strapB_03, bone2Name=dyn_strapB_02, swing1LimitDegrees=35, swing2LimitDegrees=35, twistLimitDegrees=16}
 {bone1Name=dyn_strapB_04, bone2Name=dyn_strapB_03, swing1LimitDegrees=35, swing2LimitDegrees=35, twistLimitDegrees=16}
 {bone1Name=dyn_strapB_05, bone2Name=dyn_strapB_04, swing1LimitDegrees=45, swing2LimitDegrees=35, twistLimitDegrees=16}
 {
  bone1Name=head
  bone2Name=neck_01
  swing1LimitDegrees=45
  swing1Motion=Free
  swing2LimitDegrees=45
  swing2Motion=Free
  twistLimitDegrees=45
  twistMotion=Free
 }
 {
  bone1Name=neck_01
  bone2Name=spine_05
  swing1LimitDegrees=45
  swing1Motion=Free
  swing2LimitDegrees=45
  swing2Motion=Free
  twistLimitDegrees=45
  twistMotion=Free
 }
 {bone1Name=dyn_strapA_01, bone2Name=head, swing1LimitDegrees=35, swing2LimitDegrees=35, twistLimitDegrees=16}
 {bone1Name=dyn_strapA_02, bone2Name=dyn_strapA_01, swing1LimitDegrees=45, swing2LimitDegrees=35, twistLimitDegrees=16}
 {bone1Name=dyn_strapA_03, bone2Name=dyn_strapA_02, swing1LimitDegrees=35, swing2LimitDegrees=35, twistLimitDegrees=16}
 {bone1Name=dyn_strapA_04, bone2Name=dyn_strapA_03, swing1LimitDegrees=35, swing2LimitDegrees=35, twistLimitDegrees=16}
 {bone1Name=dyn_strapA_05, bone2Name=dyn_strapA_04, swing1LimitDegrees=45, swing2LimitDegrees=35, twistLimitDegrees=16}
 {bone1Name=dyn_strapB_01, bone2Name=dyn_strapA_01, swing1LimitDegrees=35, swing2LimitDegrees=35, twistLimitDegrees=16}
 {
  bone1Name=clavicle_l
  bone2Name=spine_05
  swing1LimitDegrees=45
  swing1Motion=Free
  swing2LimitDegrees=45
  swing2Motion=Free
  twistLimitDegrees=45
  twistMotion=Free
 }
 {
  bone1Name=clavicle_r
  bone2Name=spine_05
  swing1LimitDegrees=45
  swing1Motion=Free
  swing2LimitDegrees=45
  swing2Motion=Free
  twistLimitDegrees=45
  twistMotion=Free
 }
 {
  bone1Name=spine_05
  bone2Name=spine_03
  swing1LimitDegrees=45
  swing1Motion=Free
  swing2LimitDegrees=45
  swing2Motion=Free
  twistLimitDegrees=45
  twistMotion=Free
 }
]
