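# Physics asset for the hair tips mesh of the Skull Brite hat accessory, giving collision and joint setup for its left and right dynamic hair bones.
# Single capsules cover each hair bone, the head and neck_01.
# Both hair bones constrain to the head with swing and twist limits, the right allowing a wider first swing than the left.
# The head-to-neck constraint leaves swing and twist free.
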PhysicsAsset /Game/Accessories/Hats/F_MED_Skull_Brite_Hat/Meshes/F_MED_Skull_Brite_HairTips_Physics
bodies {
 dyn_hair01_l [{shapeType=Capsule, length=4.1127, radius=2.9194, center=<-6.7922 0.8999 -0.221>, rotation=<rot 0 -154.9763 -76.032>}]
 dyn_hair01_r [{shapeType=Capsule, length=7.239, radius=2.7661, center=<9.7019 -2.2912 -0.2129>, rotation=<rot 0 -129.3826 -55.1529>}]
 head [{shapeType=Capsule, length=11.7707, radius=3.75, rotation=<rot 90 0 0>}]
 neck_01 [{shapeType=Capsule, length=7.005, radius=4.255, center=<-0.3894 1.2292 0>, rotation=<rot 0 -89.9998 90.0003>}]
}
constraints [
 {bone1Name=dyn_hair01_l, bone2Name=head, swing1LimitDegrees=25, swing2LimitDegrees=45, twistLimitDegrees=45}
 {bone1Name=dyn_hair01_r, bone2Name=head, swing1LimitDegrees=51.8571, swing2LimitDegrees=45, twistLimitDegrees=45}
 {
  bone1Name=head
  bone2Name=neck_01
  swing1LimitDegrees=45
  swing1Motion=Free
  swing2LimitDegrees=45
  swing2Motion=Free
  twistLimitDegrees=45
  twistMotion=Free
 }
]
